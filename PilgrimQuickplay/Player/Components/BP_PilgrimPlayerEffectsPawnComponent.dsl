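; Pilgrim Quickplay player pawn component that plays local cosmetic gameplay cues for overdrive deploys and other players readying up.
; On BeginPlay, where cosmetic events can run, it waits for the Pilgrim event router and listens for overdrive changes on its owner pawn, bursting the deploy cue on activation.
; A locally controlled owner also hears all readiness changes, cueing when a player not locally viewed becomes ready.
; EndPlay stops every stored listener.
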(event EventBeginPlay
  (Networking|CanExecuteCosmeticEvents
    (:True
      (bind _asfort_player_pawn (Utilities|Casting|CastToFortPlayerPawn (Components|GetOwner)))
      (Variables|Runtime|SetOwnerPlayerPawn _asfort_player_pawn)
      ;[0.5587 1 0.217] Wait for and get the event router
      (bind _eventrouter (Pilgrim|PilgrimWaitForEventRouter (Variables|Runtime|GetOwnerPlayerPawn))
        (:OnEventRouterAvailable
          (Variables|Runtime|SetEventRouter _eventrouter)
          ;[1 1 1] By using the owner as context actor, we should only get events for the current pawn. We double check just to be sure.
          (bind (_asynctaskproxy _payload) (GameplayEventRouter|StartListeningToEvent (Variables|Runtime|GetEventRouter) "/Script/PilgrimCoreRuntime.PilgrimEvent_BandState_OverdriveActiveChanged" (Variables|Runtime|GetOwnerPlayerPawn))
            (:then
              (Utilities|Array|Add (Variables|Runtime|GetEvents) _asynctaskproxy)
              (if (Pawn|IsLocallyControlled (Variables|Runtime|GetOwnerPlayerPawn))
                (bind (_asynctaskproxy_2 _payload_2) (GameplayEventRouter|StartListeningToEvent (Variables|Runtime|GetEventRouter) "/Script/PilgrimCoreRuntime.PilgrimEvent_PlayerState_ReadinessChanged")
                  (:then
                    (Utilities|Array|Add (Variables|Runtime|GetEvents) _asynctaskproxy_2))
                  (:OnEventReceived
                    (if (CallFunction|ShouldPlayReadyCue _payload_2)
                      (GameplayCue|ExecuteGameplayCue%28Burst%29Local (Variables|Runtime|GetOwnerPlayerPawn) (Variables|Default|GetOtherPlayerReadyTag) (Ability|GameplayCue|MakeGameplayCueParameters 0.0)))))))
            (:OnEventReceived
              (if (and (Utilities|Struct|BreakPilgrimEventBandStateOverdriveActiveChanged _payload) (== (.ForPlayer (Utilities|Struct|BreakPilgrimEventBandStateOverdriveActiveChanged _payload)) (Variables|Runtime|GetOwnerPlayerPawn)))
                (GameplayCue|ExecuteGameplayCue%28Burst%29Local (Variables|Runtime|GetOwnerPlayerPawn) (Variables|Default|GetODDeployTag) (Ability|GameplayCue|MakeGameplayCueParameters 1.0))))))))))

(event EventEndPlay
  (params (EndPlayReason :type EEndPlayReason%20Enum :default "Destroyed"))
  ;[0 0 0 0.5] Clean up Listeners
  (for _array_element (Variables|Runtime|GetEvents)
    (GameplayEventRouter|StopListeningtoEvent _array_element))
  (Utilities|Array|Clear (Variables|Runtime|GetEvents)))

(fn ShouldPlayReadyCue
  (params (Payload :type Pilgrim%20Event%20Player%20State%20Readiness%20Changed%20Structure%20%28by%20ref%29))
  (bind (_forplayer _bisready) (Utilities|Struct|BreakPilgrimEventPlayerStateReadinessChanged Payload))
  (return (and _bisready (not (Pawn|IsLocallyViewed _forplayer)))))
