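; Player controller component for Pilgrim Quickplay that provides song menu and quick emote helpers.
; On BeginPlay it waits for the Pilgrim event router and caches it.
; StopPlayingEmotes cancels emote-tagged abilities and broadcasts a quick-emote-stopped event through the router, while OpenSongMenu removes then re-adds the song menu scene via the Dynamic UI Manager.
; The Enhanced Input handlers for song menu and quick emote trigger no actions here.
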(event EventBeginPlay
  (bind _eventrouter (Pilgrim|PilgrimWaitForEventRouter (Components|GetOwner))
    (:OnEventRouterAvailable
      (Variables|Default|SetEvent%20Router _eventrouter))))

(event EnhancedInputActionIA_PilgrimOpenSongMenu
  (params (ActionValue :type Boolean) (ElapsedSeconds :type Float%20%28double-precision%29) (TriggeredSeconds :type Float%20%28double-precision%29) (InputAction :type Input%20Action%20Object%20Reference :default "/PilgrimCore/Input/InputActions/IA_PilgrimOpenSongMenu.IA_PilgrimOpenSongMenu")))

(event EnhancedInputActionIA_PilgrimQuickEmote_Released
  (params (ActionValue :type Boolean) (ElapsedSeconds :type Float%20%28double-precision%29) (TriggeredSeconds :type Float%20%28double-precision%29) (InputAction :type Input%20Action%20Object%20Reference :default "/PilgrimCore/Player/Abilities/IA_PilgrimQuickEmote_Released.IA_PilgrimQuickEmote_Released"))
  (bind _returnvalue (Components|GetOwner))
  (bind _returnvalue_2 (Fortnite|GetFortPlayerControllerfromActor _returnvalue))
  (bind _returnvalue_3 (Actor|GetComponentbyClass _returnvalue_2 "/Script/PilgrimCoreRuntime.PilgrimPlayerControllerComponent")))

(event EnhancedInputActionIA_PilgrimQuickEmote_Stop
  (params (ActionValue :type Boolean) (ElapsedSeconds :type Float%20%28double-precision%29) (TriggeredSeconds :type Float%20%28double-precision%29) (InputAction :type Input%20Action%20Object%20Reference :default "/PilgrimCore/Player/Abilities/IA_PilgrimQuickEmote_Stop.IA_PilgrimQuickEmote_Stop")))

(event EnhancedInputActionIA_PilgrimQuickEmote_Pressed
  (params (ActionValue :type Boolean) (ElapsedSeconds :type Float%20%28double-precision%29) (TriggeredSeconds :type Float%20%28double-precision%29) (InputAction :type Input%20Action%20Object%20Reference :default "/PilgrimCore/Player/Abilities/IA_PilgrimQuickEmote_Pressed.IA_PilgrimQuickEmote_Pressed")))

(event Custom|StopPlayingEmotes
  ;[0.5722 1 0.9768 0.5] Capture Quick Emote input actions
  ;;[0.2889 0.3106 1] Stop any currently playing emotes
  (Tags|CancelAbilitiesWithTags (Ability|GetFortAbilitySystemComponent (Fortnite|GetControllerfromActor (Components|GetOwner))) (GameplayTags|MakeLiteralGameplayTagContainer "(GameplayTags=((TagName=\"Gameplay.Action.Player.Emote\")))"))
  (bind _event_router (Variables|Default|GetEvent%20Router)
    (:then
      (GameplayEventRouter|BroadcastEvent _event_router (Utilities|Struct|MakePilgrimEventQuickEmoteStopped (Fortnite|GetFortPlayerControllerfromActor (Components|GetOwner))) 0))
    (:else)))

(fn AddSongMenuScene
  (DynamicUI|AddScene (WorldSubsystems|GetDynamicUIManager) (Variables|Default|GetSongMenuScene) (Fortnite|GetFortPlayerControllerfromActor (Components|GetOwner))))

(fn OpenSongMenu
  (CallFunction|RemoveSongMenuScene)
  (CallFunction|AddSongMenuScene))

(fn RemoveSongMenuScene
  (DynamicUI|RemoveScene (WorldSubsystems|GetDynamicUIManager) (Variables|Default|GetSongMenuScene) (Fortnite|GetFortPlayerControllerfromActor (Components|GetOwner))))
