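; Player component that gates first-time user experience (FTUE) screens for the Pilgrim Quickplay main stage, overriding hooks from a parent FTUE component.
; The modal is attempted only while the playspace state machine is in Pregame, and all screens close when that state's end event arrives.
; Non-primary players count as having seen it.
; It disables itself for touch input when the PilgrimInverse console variable is on, since that mode has its own FTUE.
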(event EventSetupAdditionalEventListeners
  (params (EventRouter :type Gameplay%20Event%20Router%20Component%20Object%20Reference))
  (bind (_asynctaskproxy _payload) (GameplayEventRouter|StartListeningToEvent EventRouter "/Script/PilgrimQPSharedRuntime.PilgrimEvent_QuickplayState_Ended")
    (:OnEventReceived
      (if (Utilities|Enum|Equal(Enum) (Utilities|Struct|BreakPilgrimEventQuickplayStateEnded _payload) "Pregame")
        (CallFunction|CloseAllFTUEScreens)))))

(fn AttemptToShowFTUEModal
  ;[0.5804 1 0.93] Only proceed if we're in the pregame state
  (if (Utilities|Enum|Equal(Enum) (Class|PilgrimQuickplayStateMachine|GetActivePilgrimStateId (Class|PilgrimQuickplayPlayspace|GetPilgrimQuickplayStateMachine (Class|SparksMusicPlayspace|GetSparksMusicPlayspaceOfClass "/PilgrimQuickplay/Playspaces/BP_PilgrimQuickplayPlayspace.BP_PilgrimQuickplayPlayspace_C" (Components|GetOwner)))) "Pregame")
    (|CallParentFunction)))

(fn HasUserSeenFTUE
  (return (or (CallFunction|GetHasSeenFTUEMainStage) (not (CallFunction|IsOwnerPrimaryPlayer)))))

(fn IsFTUEComponentEnabled
  (if (not (Inputs|IsUsingInputType "Touch" (Fortnite|GetFirstLocalFortPlayerController true)))
    (return true)
    (else
      ;[0.15 0.02022 0.0765 0.5] If player is using touch, and PilgrimInverse is on, don't use this FTUEComponent, since PilgrimInverse already has its own
      (bind _returnvalue (Development|GetConsoleVariableBoolValue "PilgrimQuickplay.EnablePilgrimInverse"))
      (return (not _returnvalue)))))

(fn SetUserHasSeenFTUE
  (params (UserHasSeenFTUE :type Boolean))
  (CallFunction|SetHasSeenFTUEMainStage UserHasSeenFTUE))
