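; Debug controller component for Pilgrim Quickplay, named for fast start, that links instrument animation layers on player pawns.
; Its LinkAnims event calls DoPreIntroAnimLinking, which gets the Quickplay playspace via the owning controller and loops over all player pawns.
; Each pawn with a valid Pilgrim instrument player state component gets its animation layer linked using that component's player animations.
; BeginPlay and Tick are empty.
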(event EventBeginPlay)

(event EventTick
  (params (DeltaSeconds :type Float%20%28single-precision%29)))

(event EventLinkAnims
  (CallFunction|DoPreIntroAnimLinking))

(fn DoPreIntroAnimLinking
  (bind _returnvalue (Class|PlayspaceLibrary|GetPlayspaceforActor self (CallFunction|GetOwningController)))
  (bind _aspilgrim_quickplay_playspace (Utilities|Casting|CastToPilgrimQuickplayPlayspace _returnvalue))
  (for _array_element (Class|SparksMusicPlayspace|GetAllPlayerPawns _aspilgrim_quickplay_playspace)
    (Utilities|IsValid (Actor|GetComponentbyClass _array_element "/PilgrimCore/Player/Components/BP_PilgrimInstrumentPlayerStateComponent.BP_PilgrimInstrumentPlayerStateComponent_C")
      (:Is%20Valid
        (Class|SparksInstrumentAnimPawnComponent|LinkInstrumentAnimationLayer (Actor|GetComponentbyClass _array_element "/PilgrimCore/Player/Components/BP_PilgrimInstrumentPlayerStateComponent.BP_PilgrimInstrumentPlayerStateComponent_C") (Variables|/PilgrimCore/Player/Components/BP_PilgrimInstrumentPlayerStateComponent.BP_PilgrimInstrumentPlayerStateComponent_C|GetPlayerAnimations (Actor|GetComponentbyClass _array_element "/PilgrimCore/Player/Components/BP_PilgrimInstrumentPlayerStateComponent.BP_PilgrimInstrumentPlayerStateComponent_C")))))))
