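; Quickplay player state component that tracks which song results player start the player's pawn occupies, as a replicated index.
; FindResultsLocationIndex takes the index from the single overlapping results start, printing an error for zero or several.
; OnRep broadcasts OnResultsLocationIndexReplicated where cosmetic events can run, and the authority resets the index to -1 when SongResults ends.
; UpdatePlayerInfoElements only unbinds from that dispatcher, flagged to move into the results widget.
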(event EventBeginPlay
  (Utilities|SwitchHasAuthority
    (:Authority
      ;[0.44 1 0.9101] Whenever the SongResults state is complete, reset the results location index
      (bind _eventrouter (Pilgrim|PilgrimWaitForEventRouter (Components|GetOwner))
        (:OnEventRouterAvailable
          (bind (_asynctaskproxy _payload) (GameplayEventRouter|StartListeningToEvent _eventrouter "/Script/PilgrimQPSharedRuntime.PilgrimEvent_QuickplayState_Ended")
            (:OnEventReceived
              (if (Utilities|Enum|Equal(Enum) (Utilities|Struct|BreakPilgrimEventQuickplayStateEnded _payload) "SongResults")
                (CallFunction|ResetResultsLocationIndex)))))))))

(event Custom|UpdatePlayerInfoElements
  (params (PlayerState :type Player%20State%20Object%20Reference) (ResultsLocationIndex :type Integer))
  (bind _returnvalue (Actor|GetComponentbyClass PlayerState "/PilgrimQuickplay/Player/Components/BP_PilgrimQuickplay_ResultsLocationPlayerStateComponent.BP_PilgrimQuickplay_ResultsLocationPlayerStateComponent_C"))
  ;[1 0 0.1205] PUT IN SONG RESULTS WBP
  (if (!= ResultsLocationIndex -1)
    (Utilities|FlowControl|Sequence
      (:then_0
        (Utilities|IsValid _returnvalue
          (:Is%20Valid
            (Default|UnbindEventfromOnResultsLocationIndexReplicated _returnvalue (EventDispatchers|CreateEvent :!event Custom|UpdatePlayerInfoElements)))))
      (:then_1
        (Utilities|FlowControl|Sequence)))))

(fn FindResultsLocationIndex
  ;[0.44 1 0.9101] Get the currently controlled pawn associated with this player state
  (bind _asfort_player_state_athena (Utilities|Casting|CastToFortPlayerStateAthena (Components|GetOwner)))
  (bind _output_get (LocalVariables|SetControlledPawn (PlayerState|GetPawn _asfort_player_state_athena)))
  ;[0.44 1 0.9101] Find all the overlapping results player starts. If there is only one, set our location. Otherwise, print a message that there's an error.
  (if (== (Utilities|Array|Length (Collision|GetOverlappingActors _output_get "/PilgrimQuickplay/Actors/PlayerStart/FortPlayerStart_PilgrimQuickplay_Results.FortPlayerStart_PilgrimQuickplay_Results_C" :!class /Script/Engine.Actor)) 1)
    (bind _asfort_player_start_pilgrim_quickplay_results (Utilities|Casting|CastToFortPlayerStart_PilgrimQuickplay_Results (Utilities|Array|Get(acopy) (Collision|GetOverlappingActors _output_get "/PilgrimQuickplay/Actors/PlayerStart/FortPlayerStart_PilgrimQuickplay_Results.FortPlayerStart_PilgrimQuickplay_Results_C" :!class /Script/Engine.Actor))))
    (Variables|Default|SetResultsLocationIndex (Math|Conversions|ToInteger(Byte) (Variables|/Script/PilgrimQuickplayRuntime.PilgrimQuickplayPlayerStart|GetPlayerStartIndex _asfort_player_start_pilgrim_quickplay_results)))
    (elif (> (Utilities|Array|Length (Collision|GetOverlappingActors _output_get "/PilgrimQuickplay/Actors/PlayerStart/FortPlayerStart_PilgrimQuickplay_Results.FortPlayerStart_PilgrimQuickplay_Results_C" :!class /Script/Engine.Actor)) 1)
      (Development|PrintString "Player pawn overlaps more than one Results Player Start, move the starts farther apart from each other.")
      (else
        (Development|PrintString "Player pawn does not overlap any Results Player Starts, something has gone wrong during the teleport")))))

(fn OnRep_ResultsLocationIndex
  (Networking|CanExecuteCosmeticEvents
    (:True
      (bind _asplayer_state (Utilities|Casting|CastToPlayerState (Components|GetOwner)))
      (Default|CallOnResultsLocationIndexReplicated _asplayer_state (Variables|Default|GetResultsLocationIndex)))))

(fn ResetResultsLocationIndex
  (Variables|Default|SetResultsLocationIndex -1))
