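; BP_PilgrimSplitscreenControllerComponent is a Pilgrim Quickplay component for a Fort player controller that provides access to the splitscreen settings screen.
; OpenSettings casts its owner to FortPlayerController, creates WBP_PilgrimSplitscreen_SettingsScreen for that controller and calls PushSettingScreen on the widget.
; BeginPlay, Tick and the IA_PilgrimSplitscreenTakeControl input event are present but have no logic, and nothing here calls OpenSettings.
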(event EventBeginPlay)

(event EventTick
  (params (DeltaSeconds :type Float%20%28single-precision%29)))

(event EnhancedInputActionIA_PilgrimSplitscreenTakeControl
  (params (ActionValue :type Boolean) (ElapsedSeconds :type Float%20%28double-precision%29) (TriggeredSeconds :type Float%20%28double-precision%29) (InputAction :type Input%20Action%20Object%20Reference :default "/PilgrimCore/Input/InputActions/IA_PilgrimSplitscreenTakeControl.IA_PilgrimSplitscreenTakeControl")))

(fn OpenSettings
  (bind _asfort_player_controller (Utilities|Casting|CastToFortPlayerController (Components|GetOwner)))
  (bind _returnvalue (UserInterface|CreateWidget "/PilgrimQuickplay/Widgets/WBP_PilgrimSplitscreen_SettingsScreen.WBP_PilgrimSplitscreen_SettingsScreen_C" _asfort_player_controller))
  (Class|WBPPilgrimSplitscreenSettingsScreen|PushSettingScreen _returnvalue))
